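# Landscape material function that turns input rock attributes into an icy, snow-covered rock, reusing the large rock for shader optimization.
# Rock opacity serves as height, driving a parallax swirl-noise detail in the shared winter ice color, with a cheaper tinted variant for mobile.
# A quality-switched height lerp blends in snow across base color, roughness, metallic and a flattened normal.
# It outputs material, height, base color and snow mask; RockTint_Color is unconnected.
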MaterialFunction /Figment_S07/LM_7_00/Athena/Environments/Landscape/MaterialFunctions/Snow/MF_Athena_IceRock_01
graph {nodes [
  n1 = MaterialFunctionCall() {materialFunction=/Figment_S07/LM_7_00/Athena/WinterArt/Materials/MF_SharedWinter}
  n2 = Constant() {r=0.05}
  n3 = LandscapeLayerCoords() {mappingScale=1}
  n4 = FunctionInput() {inputName=Rock, inputType=FunctionInput_MaterialAttributes}
  n5 = GetMaterialAttributes(None=@n4) {
   desc="Reusing Existing Large rock -  shader optimisation"
   attributeGetTypes [#1, #2, #3]
  }
  n6 = ComponentMask(None=@n5.Opacity) {a=true}
  n7 = TextureSample(UVs=BumpOffset(Coordinate=Multiply(A=@n2, B=@n3), Height=@n6, HeightRatioInput=Constant() {r=2})) {
   samplerSource=SSM_Wrap_WorldGroupSettings
   samplerType=SAMPLERTYPE_LinearGrayscale
   texture=/Figment_S07/LM_7_00/Effects/Fort_Effects/BluGlow/Violentium/T_SwirlNoise_N_height
  }
  n8 = Saturate(None=MaterialFunctionCall!(Cutoff Value=Constant() {r=1.8916}, Lerp Value=Constant() {r=-0.0952}, Gradient=@n6) {materialFunction=/Engine/Functions/Engine_MaterialFunctions02/SmoothThreshold}.Result)
  n9 = Multiply(
   A=VectorParameter() {parameterName=IceModulateBaseColo, defaultValue=<rgba 0.7135 0.703 0.6715 1>}.RGB
   B=LinearInterpolate(A=@n1.IceColor, B=Add(A=@n7.R, B=@n1.IceColor), Alpha=@n8)
  )
  n10 = Multiply(
   A=MaterialFunctionCall(In=@n6, Contrast=Constant() {r=-0.25}) {materialFunction=&CheapContrast}.Result
   B=VectorParameter() {parameterName=Mobile_IceColor, defaultValue=<rgba 0.4531 0.7539 1 1>}.RGB
  )
  n11 = Constant() {r=0.6952}
  n12 = Constant() {r=0.75}
  n13 = MaterialFunctionCall(Transition Phase=QualitySwitch(Default=@n11, Low=@n12, High=@n11, Medium=@n12), Height Texture=@n6, Contrast=Constant() {r=-2.5}) {materialFunction=&HeightLerp}
  n14 = LinearInterpolate(
   A=FeatureLevelSwitch(Default=@n9, ES2=@n10, ES3_1=@n10, SM4_REMOVED=@n9, SM5=@n9, SM6=@n9)
   B=ScalarParameter() {defaultValue=0.8, parameterName=Snow}
   Alpha=FunctionOutput(None=@n13.Alpha) {outputName=EmissiveMask}
  )
  n15 = ScalarParameter() {defaultValue=0.15, parameterName=Ice_R3}
  n16 = ScalarParameter() {defaultValue=0.2, parameterName=Ice_R2}
  n17 = SetMaterialAttributes!(
   MaterialAttributes
   BaseColor=@n14
   Specular
   Roughness=LinearInterpolate(A=LinearInterpolate(A=@n15, B=@n16, Alpha=@n8), B=ScalarParameter() {defaultValue=0.5, parameterName=SnowR}, Alpha=@n13.Alpha)
   Metallic=LinearInterpolate(A=Constant() {r=0.25}, Alpha=@n13.Alpha) {constB=0}
   Normal=LinearInterpolate(A=@n5.Normal, B=Constant3Vector() {constant=<rgba 0 0 1 0>}, Alpha=@n13.Alpha)
   Opacity=BreakMaterialAttributes!(Attr=@n4).Opacity
  ) {attributeSetTypes [
    #1
    #4
    #5
    #6
    #3
    #2
   ]}
  FunctionOutput(None=@n17) {outputName=Material}
  FunctionOutput(None=@n6) {outputName=Height}
  FunctionOutput(None=@n14) {outputName=BaseColor_Only}
  FunctionOutput() {outputName=RockTint_Color}
 ]}
